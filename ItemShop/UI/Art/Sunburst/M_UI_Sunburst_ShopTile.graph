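# UI material for an Item Shop tile sunburst background, with blue, cyan and purple radial gradients.
# A tiled CRT texture, animated through a flow map driven by panning, rotating noise, blends cyan highlights into the gradients.
# A stars texture adds white, and parameterized horizontal and vertical edge masks darken the borders.
# A standalone sample of a ShopTile sunburst texture sits beside this procedural chain, with neither shown wired to an output.
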Material /ItemShop/UI/Art/Sunburst/M_UI_Sunburst_ShopTile
graph {
 properties {bUsedWithStaticMesh=false, materialDomain=MD_UI, pixelDepthOffsetMode=PDOM_Legacy, usedShadingModels=Default Lit}
 nodes [
  n1 = MaterialFunctionCall() {materialFunction=&MF_UI_UVScaling}
  n2 = Abs(None=ConstantBiasScale(None=@n1.Scaled To Fill) {bias=-0.5, scale=2})
  n3 = Multiply(
   A=Subtract(A=ComponentMask(None=@n2) {g=true}, B=ScalarParameter() {defaultValue=0.6, parameterName=Subtract})
   B=ScalarParameter() {defaultValue=1, parameterName=Falloff}
  )
  n4 = Abs(None=ConstantBiasScale(None=@n1.Scaled To Fill) {bias=-0.5, scale=2})
  n5 = Multiply(
   A=Subtract(A=ComponentMask(None=@n4) {r=true}, B=ScalarParameter() {defaultValue=0.6, parameterName=SubtractH})
   B=ScalarParameter() {defaultValue=0.5, parameterName=FalloffH}
  )
  n6 = LinearInterpolate(
   A=Multiply(A=Saturate(None=@n3), B=ScalarParameter() {defaultValue=3.5, parameterName=Intensity})
   B=Multiply(A=Saturate(None=@n5), B=ScalarParameter() {defaultValue=3.5, parameterName=IntensityH})
  )
  n7 = TextureSample() {texture=/OfferCatalog/Art/Sunburst/T_UI_SunburstStars}
  n8 = Constant3Vector() {constant=<rgba 1 1 1 0>}
  n9 = LinearInterpolate(
   A=Constant3Vector() {constant=<rgba 0 0.0942 0.6476 0>}
   B=Constant3Vector() {constant=<rgba 0 0.8632 0.956 0>}
   Alpha=MaterialFunctionCall(GradientStart=Constant(), GradientEnd=Constant() {r=0.8747}) {materialFunction=&MF_UI_RadialGradient}.Smooth
  )
  n10 = Multiply(
   A=MaterialFunctionCall() {materialFunction=&MF_UI_UVScaling}.Scaled To Fill
   B=Constant2Vector() {g=42, r=95}
  ) {constB=75}
  n11 = Multiply(A=Multiply(A=Constant2Vector() {g=0.75, r=0.9}, B=TextureCoordinate()), B=ScalarParameter() {defaultValue=-0.8, parameterName=DupeScaler})
  n12 = Multiply(A=Time(), B=ScalarParameter() {defaultValue=0.5, parameterName=TimeFactor})
  n13 = Time()
  n14 = Constant() {r=0.001}
  n15 = TextureSample(UVs=Rotator(Coordinate=Panner(Coordinate=@n11, Time=@n12) {speedX=0.025, speedY=0.005}, Time=Multiply(A=@n13, B=@n14))) {samplerType=SAMPLERTYPE_Alpha, texture=/OfferCatalog/Art/Sunburst/T_UI_SunburstMovement}
  n16 = ScalarParameter() {defaultValue=1.888, parameterName=DistortionScaler}
  n17 = MaterialFunctionCall!(
   Texture 2D=TextureObject() {texture=/OfferCatalog/Art/Sunburst/T_UI_CRTTexture}
   Texture UVs=MaterialFunctionCall(UVs=Frac(None=@n10), Texture Scale=Constant2Vector() {g=0.5, r=4}) {materialFunction=&ScaleUVsByCenter}.UVs
   FlowMap Texture 2D=TextureObject() {texture=/OfferCatalog/Art/Sunburst/T_UI_CirclesFlowmap}
   FlowMap UVs=NamedRerouteDeclaration(None=Multiply(A=@n15.R, B=@n16)) {name=UV-Distortion, nodeColor=<rgba 0 0 0 1>}
   FlowIntensity=Constant() {r=1}
   FlowSpeed=Constant() {r=0.15}
  ) {materialFunction=/FortUILibrary/MaterialFunctions/Effects/MF_UI_FlowMap}
  n18 = ComponentMask(None=@n17.Result) {r=true}
  n19 = ComponentMask(None=@n17.Result) {g=true}
  n20 = Multiply(A=Max(A=Max(A=@n18, B=@n19), B=ComponentMask(None=@n17.Result) {b=true})) {constB=1.25}
  n21 = MaterialFunctionCall(UVs=Add(A=TextureCoordinate() {coordinateIndex=1}, B=Constant2Vector() {g=0.4, r=-0.5}), Texture Scale=Constant() {r=0.5}) {materialFunction=&ScaleUVsByCenter}
  n22 = Constant()
  n23 = Constant() {r=1.5}
  n24 = Constant3Vector() {constant=<rgba 0 0.0054 0.3628 0>}
  n25 = Multiply(A=LinearInterpolate(
    A=@n24
    B=Constant3Vector() {constant=<rgba 0.4945 0 1 0>}
    Alpha=MaterialFunctionCall(UV=@n21.UVs, GradientStart=@n22, GradientEnd=@n23) {materialFunction=&MF_UI_RadialGradient}.Smooth
   )) {constB=3}
  n26 = LinearInterpolate(
   A=@n24
   B=Constant3Vector() {constant=<rgba 0 0.3876 1 0>}
   Alpha=MaterialFunctionCall(GradientStart=Constant(), GradientEnd=Constant() {r=1}) {materialFunction=&MF_UI_RadialGradient}.Smooth
  )
  n27 = MaterialFunctionCall(UVs=Add(A=TextureCoordinate() {coordinateIndex=1}, B=Constant2Vector() {g=-0.3, r=-0.25}), Texture Scale=Constant() {r=0.5}) {materialFunction=&ScaleUVsByCenter}
  n28 = Constant()
  n29 = Constant() {r=1.2305}
  n30 = Multiply(A=LinearInterpolate(
    B=Constant3Vector() {constant=<rgba 0.4945 0 1 0>}
    Alpha=MaterialFunctionCall(UV=@n27.UVs, GradientStart=@n28, GradientEnd=@n29) {materialFunction=&MF_UI_RadialGradient}.Smooth
   )) {constB=3}
  n31 = Add(A=TextureCoordinate() {coordinateIndex=1}, B=Constant2Vector() {g=0.3, r=0.5})
  n32 = Constant() {r=0.5}
  n33 = LinearInterpolate(
   B=Constant3Vector() {constant=<rgba 0.4945 0 1 0>}
   Alpha=MaterialFunctionCall(
    UV=MaterialFunctionCall(UVs=@n31, Texture Scale=@n32) {materialFunction=&ScaleUVsByCenter}.UVs
    GradientStart=Constant()
    GradientEnd=Constant() {r=1.5133}
   ) {materialFunction=&MF_UI_RadialGradient}.Smooth
  )
  n34 = LinearInterpolate(
   A=LinearInterpolate(A=LinearInterpolate(A=@n25, B=@n26), B=LinearInterpolate(A=@n30, B=@n26))
   B=LinearInterpolate(A=LinearInterpolate(A=@n33, B=@n26), B=@n26, Alpha=@n26)
  )
  Multiply(
   A=LinearInterpolate(A=Constant3Vector(), Alpha=OneMinus(None=@n6)) {constAlpha=0}
   B=LinearInterpolate(A=LinearInterpolate(A=@n7.R, B=@n8, Alpha=@n7.R), Alpha=LinearInterpolate(A=@n34, B=@n9, Alpha=@n20) {constA=0.5})
  )
  TextureSample(UVs=MaterialFunctionCall() {materialFunction=&GetUserInterfaceUV}.Normalized UV) {texture=/OfferCatalog/Art/Sunburst/T_UI_Sunburst_ShopTile}
 ]
}
