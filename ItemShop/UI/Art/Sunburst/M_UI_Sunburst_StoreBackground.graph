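# UI-domain material for the item shop's sunburst store background, built around the T_UI_Sunburst_StoreBG texture.
# A single texture sample reads that texture using the normalized UV from GetUserInterfaceUV.
# A UV scaling function and a texture coordinate node are also present but feed nothing.
# The graph shows no link from the sample to a material output.
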Material /ItemShop/UI/Art/Sunburst/M_UI_Sunburst_StoreBackground
graph {
 properties {bUsedWithStaticMesh=false, materialDomain=MD_UI, pixelDepthOffsetMode=PDOM_Legacy, usedShadingModels=Default Lit}
 nodes [
  MaterialFunctionCall() {materialFunction=&MF_UI_UVScaling}
  TextureCoordinate()
  TextureSample(UVs=MaterialFunctionCall() {materialFunction=&GetUserInterfaceUV}.Normalized UV) {texture=&/T_UI_Sunburst_StoreBG}
 ]
}
